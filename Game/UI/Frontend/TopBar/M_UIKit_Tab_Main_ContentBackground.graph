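# UI material for the content background of a main tab in the frontend top bar, translucent with a rounded-corner mask function.
# A Slate post buffer feeds backplate functions as blur input, producing a regular backplate dimmed or blended by a lerp and a disabled variant.
# Modular button state functions blend these into an interaction color.
# The Is Selectable static switch adds selected-state blending on top of the base states.
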Material /Game/UI/Frontend/TopBar/M_UIKit_Tab_Main_ContentBackground
graph {
 properties {
  bUsedWithStaticMesh=false
  blendMode=BLEND_Translucent
  materialDomain=MD_UI
  pixelDepthOffsetMode=PDOM_Legacy
  usedShadingModels=Default Lit
 }
 nodes [
  MaterialFunctionCall(Corner Radius=Constant() {r=12}, Radius in Pixels=Constant() {r=1}) {materialFunction=&MF_UI_CoreUI_RoundedMask}
  NamedRerouteDeclaration(None=MaterialFunctionCall!("UVs (Optional)") {materialFunction=/Engine/Functions/UserInterface/GetSlatePost0}.LinearRGB) {name=Buffer RGB, nodeColor=<rgba 0 0 0 1>}
  n1 = NamedRerouteUsage()
  NamedRerouteDeclaration(None=LinearInterpolate(A=MaterialFunctionCall!(Post Buffer Blur=@n1) {materialFunction=/FortUILibrary/MaterialFunctions/CoreUI/MF_UI_CoreUI_BackplateA}.Result) {constAlpha=0.25}) {name=Backplate A, nodeColor=<rgba 0 0 0 1>}
  n2 = NamedRerouteDeclaration(None=NamedRerouteUsage()) {name=Regular CTA, nodeColor=<rgba 0 0 0 1>}
  n3 = MaterialFunctionCall!(
   Idle=@n2
   Idle Opacity
   Hovered=@n2
   Hovered Opacity
   Focused=@n2
   Focused Opacity
   Pressed=@n2
   Pressed Opacity
   Disabled=NamedRerouteUsage()
   Disabled Opacity
  ) {materialFunction=/FortUILibrary/Systems/ModularButton/MaterialFunctions/MF_UIKit_ButtonStatesBase_Lerp}
  NamedRerouteDeclaration(None=StaticSwitchParameter(
    True=MaterialFunctionCall!(
     Base States=@n3.Color
     Base States Opacity
     Selected_Idle
     Selected_Idle Opacity
     Selected_Hovered
     Selected_Hovered Opacity
     Selected_Focused
     Selected_Focused Opacity
     Selected_Pressed
     Selected_Pressed Opacity
     Selected_Disabled
     Selected_Disabled Opacity
    ) {materialFunction=/FortUILibrary/Systems/ModularButton/MaterialFunctions/MF_UIKit_ButtonStatesSelectable_Lerp}.Color
    False=@n3.Color
   ) {parameterName=Is Selectable}) {name=Interaction Color, nodeColor=<rgba 0 0 0 1>}
  NamedRerouteDeclaration(None=MaterialFunctionCall!(Post Buffer Blur=NamedRerouteUsage()) {materialFunction=/FortUILibrary/MaterialFunctions/CoreUI/MF_UI_CoreUI_BackplateDisabled}.Result) {name=Backplate Disabled, nodeColor=<rgba 0 0 0 1>}
  NamedRerouteUsage()
 ]
}
