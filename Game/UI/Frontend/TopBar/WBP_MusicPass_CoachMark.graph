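# Music Pass onboarding coach mark for the frontend top bar, telling players that XP earned anywhere progresses the pass.
# Wraps WBP_UIKit_CoachMark in a FortSidebarOnboardTooltipWidget, setting its button style and platform-dependent top padding on PreConstruct.
# Showing the tooltip activates the coach mark; the "Got it" click, tooltip end or focus-path loss deactivates the widget.
# Deactivation reports dismissal and close-animation completion to the parent, then deactivates the coach mark.
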WidgetBlueprint /Game/UI/Frontend/TopBar/WBP_MusicPass_CoachMark
graphs [EventGraph {nodes [
   "Button Other Action Clicked (WBP_UIKit_CoachMark)"() {is=title K2Node_ComponentBoundEvent}
   > n1 = HandleOtherActionClicked(self:Fort Sidebar Onboard Tooltip Widget Object Reference)
   DeactivateWidget(execute=[@n1.then, @n3.then, @n2.then])
   Event PreConstruct()
   > SetButtonType!(
    execute
    self:WBP UIKit Button Regular Object Reference=Get OtherActionButton!(self:WBP UIKit Coach Mark Base Object Reference=Get WBP_UIKit_CoachMark() -> (WBP_UIKit_CoachMark:WBP UIKit Coach Mark Object Reference)) -> (OtherActionButton:WBP UIKit Button Regular Object Reference)
    ButtonType:E_UI_CTAButtonType Enum=NewEnumerator1
   ) -> (then)
   > SetPadding!(
    execute
    self:Size Box Slot Object Reference=SlotAsSizeBoxSlot!(Widget:Widget Object Reference=Get WBP_UIKit_CoachMark()) -> (ReturnValue:Size Box Slot Object Reference)
    InPadding_Left:float=0
    InPadding_Top:float=Select(Option 0:float=70, Option 1:float=145, Index=IsMobileGame()) -> (ReturnValue:float)
    InPadding_Right:float=0
    InPadding_Bottom:float=0
   ) -> (then)
   Event On Deactivated()
   > HandleDismissClicked(self:Fort Sidebar Onboard Tooltip Widget Object Reference)
   > NotifyCloseAnimationFinished(self:Fort Sidebar Onboard Tooltip Widget Object Reference)
   > DeactivateWidget(self=Get WBP_UIKit_CoachMark())
   n2 = Event OnEndTooltip() {is=title K2Node_Event}
   n3 = Event OnRemovedFromFocusPath!() -> (OutputDelegate, then, InFocusEvent:Focus Event Structure) {is=title K2Node_Event}
   Event OnShowTooltip() {is=title K2Node_Event}
   > ActivateWidget(self=Get WBP_UIKit_CoachMark())
  ]}]
variables []
widgets {
 info {
  inheritedWidgetCount=2
  parentClass=/Script/FortniteUI.FortSidebarOnboardTooltipWidget
  rootWidgetClass=&CommonVisualAttachment
  widgetCount=4
 }
 widgets [
  CommonVisualAttachment {widgetClassPath=&CommonVisualAttachment, bIsVariable=true, widget=CommonVisualAttachment, properties {contentAnchor=<0.5 0>}}
  WBP_UIKit_CoachMark {
   widgetClassPath=/FortUILibrary/UIKit/CoachMark/WBP_UIKit_CoachMark
   bIsVariable=true
   parent=CommonVisualAttachment
   widget=WBP_UIKit_CoachMark
   properties {
    arrowPosition=Top_Center
    bIsFocusable=true
    justificationHeader=Center
    navigation=&self:WidgetTree.WBP_UIKit_CoachMark.WidgetNavigation_0
    shouldShowDismissButton=false
    textDescription="Earn XP anywhere to progress the Music Pass and claim rewards!"
    textHeader=""
    textOtherActionButton=Got it
    visibility=SelfHitTestInvisible
    in Decorator Classes []
    input Action Other {dataTable=&InputActionDataTable, rowName=GenericForward}
   }
   slot {class=SizeBoxSlot, properties {horizontalAlignment=HAlign_Left, verticalAlignment=VAlign_Top}}
  }
  Dismiss_Button {widgetClassPath=&CommonButtonLegacy, bInherited=true, widget=null}
  OtherAction_Button {widgetClassPath=&CommonButtonLegacy, bInherited=true, widget=null}
 ]
}
