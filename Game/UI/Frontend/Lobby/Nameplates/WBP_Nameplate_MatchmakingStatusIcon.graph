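# Lobby nameplate widget that shows a player's matchmaking status as a small icon on a circular backplate.
# UpdateStatusIcon hides the widget for None and NotReady, sets the icon material from a status-to-material map, and shows the flash image only for Ready.
# Ready plays a ready-up animation that scales backplate and icon while animating flash material parameters; other statuses play a fade-in.
# Two post-buffer update widgets with blur updaters bracket the visuals.
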WidgetBlueprint /Game/UI/Frontend/Lobby/Nameplates/WBP_Nameplate_MatchmakingStatusIcon
animations [
 Anim_Intro_MatchmakingStatusUpdate {
  binding_count=1
  length_seconds=1.0017
  object=&self:Anim_Intro_MatchmakingStatusUpdate
  sequence {bindings [Overlay {
     id=#1
     possessed_class=Overlay
     tracks [{
       type=MovieSceneFloatTrack
       display_name=Render Opacity
       sections [{type=MovieSceneFloatSection, channels [None {type=MovieSceneScriptingFloatChannel, keys [[0,0],[20.0343,1]]}]}]
      }]
    }]}
 }
 Anim_ReadyUp {
  binding_count=3
  length_seconds=0.7541
  object=&self:Anim_ReadyUp
  sequence {bindings [
    Image_Ready_Up {
     id=#2
     possessed_class=Image
     tracks [{
       type=MovieSceneWidgetMaterialTrack
       sections [{
         type=MovieSceneParameterSection
         end_frame=15
         start_frame=0
         channels [
          Beam Height {type=MovieSceneScriptingFloatChannel, keys [[0,0.08],[3,0.0717],[7.5407,0]]}
          Beam Width {type=MovieSceneScriptingFloatChannel, keys [[0,0],[3,1]]}
          Beam Opacity {type=MovieSceneScriptingFloatChannel, keys [[0,1],[15.0817,0]]}
          Ring -Speed {type=MovieSceneScriptingFloatChannel, keys [[0,0],[3,0],[15.0817,1]]}
          Glow Strength/Opacity {type=MovieSceneScriptingFloatChannel, keys [[0,0],[5.2787,1.0031],[15.0817,0]]}
          Ring Max Size {type=MovieSceneScriptingFloatChannel, const [0.6,1]}
          Glow Radius {type=MovieSceneScriptingFloatChannel, keys [[0,0.3],[9.049,0.4]]}
         ]
        }]
      }]
    }
    WBP_UIKit_Backplate {
     id=#3
     possessed_class=WBP_UIKit_Backplate_C
     tracks [{
       type=MovieScene2DTransformTrack
       sections [{
         type=MovieScene2DTransformSection
         end_frame=15
         start_frame=0
         channels [
          Scale.X {type=MovieSceneScriptingFloatChannel, keys [[0,0],[7,1],[15.0817,1]]}
          Scale.Y {type=MovieSceneScriptingFloatChannel, keys [[0,0],[7,1],[15.0817,1]]}
         ]
        }]
      }]
    }
    Image_Icon {
     id=#4
     possessed_class=Image
     tracks [{
       type=MovieScene2DTransformTrack
       sections [{
         type=MovieScene2DTransformSection
         end_frame=15
         start_frame=0
         channels [
          Translation.X {type=MovieSceneScriptingFloatChannel, const [0,1]}
          Translation.Y {type=MovieSceneScriptingFloatChannel, const [0,1]}
          Angle {type=MovieSceneScriptingFloatChannel, const [0,1]}
          Scale.X {type=MovieSceneScriptingFloatChannel, keys [[0,0],[7,1],[15,1]]}
          Scale.Y {type=MovieSceneScriptingFloatChannel, keys [[0,0],[7,1],[15,1]]}
          Shear.X {type=MovieSceneScriptingFloatChannel, const [0,1]}
          Shear.Y {type=MovieSceneScriptingFloatChannel, const [0,1]}
         ]
        }]
      }]
    }
   ]}
 }
]
graphs [
 EventGraph {links [], nodes []}
 UpdateStatusIcon {nodes [
   UpdateStatusIcon!() -> (then, CurrentMatchmakingStatus:EPedestalNameplateMatchmakingStatus Enum=None) {is=title K2Node_FunctionEntry}
   > Set Visibility(Visibility=Select!(
     None:ESlateVisibility Enum=Collapsed
     Ready:ESlateVisibility Enum=SelfHitTestInvisible
     NotReady:ESlateVisibility Enum=Collapsed
     Playing:ESlateVisibility Enum=SelfHitTestInvisible
     Disconnected:ESlateVisibility Enum=SelfHitTestInvisible
     Downloading:ESlateVisibility Enum=SelfHitTestInvisible
     Index:EPedestalNameplateMatchmakingStatus Enum=Get CurrentMatchmakingStatus() -> (CurrentMatchmakingStatus:EPedestalNameplateMatchmakingStatus Enum=None)
    ) -> (ReturnValue))
   > SetBrushFromMaterial(
    self=Get Image_Icon() -> (Image_Icon:Image Object Reference)
    Material=Find(
     TargetMap:Map of EPedestalNameplateMatchmakingStatus Enums to Material Instance Object References=Get MatchmakingStatusIconsMap() -> (MatchmakingStatusIconsMap:Map of EPedestalNameplateMatchmakingStatus Enums to Material Instance Object References=None)
     Key:"EPedestalNameplateMatchmakingStatus Enum (by ref)"=Get CurrentMatchmakingStatus() -> (CurrentMatchmakingStatus=None)
    ) -> (Value:Material Instance Object Reference)
   )
   > Set Visibility(
    Visibility=Select(
     None=Collapsed
     Ready=SelfHitTestInvisible
     NotReady=Collapsed
     Playing=Collapsed
     Disconnected=Collapsed
     Downloading=Collapsed
     Index=Get CurrentMatchmakingStatus() -> (CurrentMatchmakingStatus=None)
    )
    self=Get Image_Ready_Up() -> (Image_Ready_Up:Image Object Reference)
   )
   n1 = Get Anim_Intro_MatchmakingStatusUpdate() -> (Anim_Intro_MatchmakingStatusUpdate:Widget Animation Object Reference)
   > PlayAnimation(InAnimation=Select(
     None:Widget Animation Object Reference=@n1
     Ready:Widget Animation Object Reference=Get Anim_ReadyUp() -> (Anim_ReadyUp:Widget Animation Object Reference)
     NotReady:Widget Animation Object Reference=@n1
     Playing:Widget Animation Object Reference=@n1
     Disconnected:Widget Animation Object Reference=@n1
     Downloading:Widget Animation Object Reference=@n1
     Index=Get CurrentMatchmakingStatus() -> (CurrentMatchmakingStatus=None)
    ) -> (ReturnValue:Widget Animation Object Reference))
  ]}
]
variables [MatchmakingStatusIconsMap {default {
   disconnected=/FortUILibrary/Textures/MaterialInstances/MI_UIKit_Icon_Disconnected
   downloading=/FortUILibrary/Textures/MaterialInstances/MI_UIKit_Icon_Download
   none=null
   notReady=null
   playing=/FortUILibrary/Textures/MaterialInstances/MI_UIKit_Icon_Controller
   ready=/Game/UI/Frontend/Lobby/Nameplates/Materials/MI_UI_Nameplate_Icon_Online_Green
  }}]
widgets {
 info {inheritedWidgetCount=0, parentClass=&UserWidget, rootWidgetClass=&Overlay, widgetCount=7}
 widgets [
  Overlay {widgetClassPath=&Overlay, widget=Overlay}
  PostBufferUpdate_100 {
   widgetClassPath=/Script/UMG.PostBufferUpdate
   parent=Overlay
   widget=PostBufferUpdate_100
   properties {
    visibility=SelfHitTestInvisible
    updateBufferInfos [{
      bufferToUpdate=ESlatePostRT_0
      postParamUpdater=&self:WidgetTree.PostBufferUpdate_100.PostBufferBlurUpdater_0
     }]
   }
   slot {class=OverlaySlot, properties {horizontalAlignment=HAlign_Fill, verticalAlignment=VAlign_Fill}}
  }
  Min_Size {
   widgetClassPath=&Spacer
   parent=Overlay
   widget=Min_Size
   properties {size=<24 24>}
   slot {class=OverlaySlot, properties {horizontalAlignment=HAlign_Fill, verticalAlignment=VAlign_Fill}}
  }
  WBP_UIKit_Backplate {
   widgetClassPath=&WBP_UIKit_Backplate
   parent=Overlay
   widget=WBP_UIKit_Backplate
   properties {brightness=Obscure, cornerRadius=Circle}
   slot {class=OverlaySlot, properties {horizontalAlignment=HAlign_Fill, verticalAlignment=VAlign_Fill}}
  }
  Image_Icon {
   widgetClassPath=&Image
   bIsVariable=true
   parent=Overlay
   widget=Image_Icon
   properties {
    visibility=HitTestInvisible
    brush {
     bIsDynamicallyLoaded=false
     drawAs=Image
     imageType=FullColor
     mirroring=NoMirror
     resourceName=None
     resourceObject=&MI_UI_Nameplate_Icon_Online_Green
     tiling=NoTile
     imageSize=<20 20>
     margin=<ltrb 0 0 0 0>
     outlineSettings {
      bUseBrushTransparency=false
      roundingType=HalfHeightRadius
      width=0
      color {colorUseRule=UseColor_Specified, specifiedColor=<rgba 0 0 0 0>}
      cornerRadii=<0 0 0 0>
     }
     tintColor {colorUseRule=UseColor_Specified, specifiedColor=<rgba 1 1 1 1>}
     uVRegion {bIsValid=false, max=<0 0>, min=<0 0>}
    }
   }
   slot {class=OverlaySlot, properties {horizontalAlignment=HAlign_Center, verticalAlignment=VAlign_Center}}
  }
  Image_Ready_Up {
   widgetClassPath=&Image
   bIsVariable=true
   parent=Overlay
   widget=Image_Ready_Up
   properties {
    visibility=Collapsed
    brush {
     bIsDynamicallyLoaded=false
     drawAs=Image
     imageType=FullColor
     mirroring=NoMirror
     resourceName=None
     resourceObject=/Game/UI/Frontend/Lobby/Nameplates/Materials/MI_UI_NamePlate_ReadyUpFlash
     tiling=NoTile
     imageSize=<24 24>
     margin=<ltrb 0 0 0 0>
     outlineSettings {
      bUseBrushTransparency=false
      roundingType=HalfHeightRadius
      width=0
      color {colorUseRule=UseColor_Specified, specifiedColor=<rgba 0 0 0 0>}
      cornerRadii=<0 0 0 0>
     }
     tintColor {colorUseRule=UseColor_Specified, specifiedColor=<rgba 1 1 1 1>}
     uVRegion {bIsValid=false, max=<0 0>, min=<0 0>}
    }
    renderTransform {angle=0, scale=<5.3333 5.3333>, shear=<0 0>, translation=<0 0>}
   }
   slot {class=OverlaySlot, properties {horizontalAlignment=HAlign_Center, verticalAlignment=VAlign_Center}}
  }
  PostBufferUpdate_55 {
   widgetClassPath=&PostBufferUpdate
   parent=Overlay
   widget=PostBufferUpdate_55
   properties {
    visibility=SelfHitTestInvisible
    updateBufferInfos [{
      bufferToUpdate=ESlatePostRT_0
      postParamUpdater=&self:WidgetTree.PostBufferUpdate_55.PostBufferBlurUpdater_0
     }]
   }
   slot {class=OverlaySlot, properties {horizontalAlignment=HAlign_Fill, verticalAlignment=VAlign_Fill}}
  }
 ]
}
