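# MTXButton is the frontend top bar button widget showing a V-Bucks icon and numeric amount, built on FortMTXButton.
# Construct sets size for mobile, then picks backing material and amount color by whether mobile and Bacchus frontend are both active.
# That same check selects Hover or Hover_Mobile, which animate colors and a material parameter.
# Hover events play it forward or reverse, replaced by press and release on touch.
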WidgetBlueprint /Game/UI/Frontend/TopBar/MTXButton
animations [
 Hover {
  binding_count=3
  length_seconds=0.2
  object=&self:Hover
  sequence {bindings [
    AmountColor {
     id=#1
     possessed_class=Border
     tracks [{
       type=MovieSceneColorTrack
       display_name=Content Color and Opacity
       sections [{
         type=MovieSceneColorSection
         end_frame=4
         start_frame=0
         channels [
          Color.R {type=MovieSceneScriptingFloatChannel, keys [[0,0],[1,0.1866],[2,0.5139],[3,0.8438],[4,1]]}
          Color.G {type=MovieSceneScriptingFloatChannel, thinned_from=5, keys [[0,0.6667],[1,0.7289],[3,0.9479],[4,1]]}
          Color.B {type=MovieSceneScriptingFloatChannel}
          Color.A {type=MovieSceneScriptingFloatChannel}
         ]
        }]
      }]
    }
    ButtonBacking {
     id=#2
     possessed_class=Image
     tracks [{
       type=MovieSceneWidgetMaterialTrack
       sections [{
         type=MovieSceneParameterSection
         end_frame=4
         start_frame=0
         channels [AnimateSelection {type=MovieSceneScriptingFloatChannel, keys [[0,0],[4,1]]}]
        }]
      }]
    }
    VBuckTint {
     id=#3
     possessed_class=Border
     tracks [{
       type=MovieSceneColorTrack
       display_name=Content Color and Opacity
       sections [{
         type=MovieSceneColorSection
         end_frame=4
         start_frame=0
         channels [
          Color.R {type=MovieSceneScriptingFloatChannel, keys [[0,0.675],[4,1]]}
          Color.G {type=MovieSceneScriptingFloatChannel, keys [[0,0.825],[4,1]]}
          Color.B {type=MovieSceneScriptingFloatChannel, keys [[0,0.9],[4,1]]}
          Color.A {type=MovieSceneScriptingFloatChannel, const [1,1]}
         ]
        }]
      }]
    }
   ]}
 }
 Hover_Bounce {
  binding_count=3
  length_seconds=0.2
  object=&self:Hover_Bounce
  sequence {bindings [
    AmountColor {
     id=#1
     possessed_class=Border
     tracks [{
       type=MovieSceneColorTrack
       display_name=Content Color and Opacity
       sections [{
         type=MovieSceneColorSection
         end_frame=4
         start_frame=0
         channels [
          Color.R {type=MovieSceneScriptingFloatChannel, keys [[0,0],[4,1]]}
          Color.G {type=MovieSceneScriptingFloatChannel, keys [[0,0.6667],[4,1]]}
          Color.B {type=MovieSceneScriptingFloatChannel}
          Color.A {type=MovieSceneScriptingFloatChannel}
         ]
        }]
      }]
    }
    ButtonBacking {
     id=#2
     possessed_class=Image
     tracks [{
       type=MovieSceneWidgetMaterialTrack
       sections [{
         type=MovieSceneParameterSection
         end_frame=4
         start_frame=0
         channels [AnimateSelection {type=MovieSceneScriptingFloatChannel, keys [[0,0],[4,1]]}]
        }]
      }]
    }
    Overlay_0 {
     id=#4
     possessed_class=Overlay
     tracks [{
       type=MovieScene2DTransformTrack
       sections [{
         type=MovieScene2DTransformSection
         end_frame=4
         start_frame=0
         channels [
          Translation.X {type=MovieSceneScriptingFloatChannel}
          Translation.Y {type=MovieSceneScriptingFloatChannel}
          Angle {type=MovieSceneScriptingFloatChannel}
          Scale.X {type=MovieSceneScriptingFloatChannel, keys [[0,1],[1,1.05],[3,1]]}
          Scale.Y {type=MovieSceneScriptingFloatChannel, keys [[0,1],[1,1.05],[3,1]]}
          Shear.X {type=MovieSceneScriptingFloatChannel}
          Shear.Y {type=MovieSceneScriptingFloatChannel}
         ]
        }]
      }]
    }
   ]}
 }
 Hover_Grow {
  binding_count=3
  length_seconds=0.2
  object=&self:Hover_Grow
  sequence {bindings [
    AmountColor {
     id=#1
     possessed_class=Border
     tracks [{
       type=MovieSceneColorTrack
       display_name=Content Color and Opacity
       sections [{
         type=MovieSceneColorSection
         end_frame=4
         start_frame=0
         channels [
          Color.R {type=MovieSceneScriptingFloatChannel, keys [[0,0],[4,1]]}
          Color.G {type=MovieSceneScriptingFloatChannel, keys [[0,0.6667],[4,1]]}
          Color.B {type=MovieSceneScriptingFloatChannel}
          Color.A {type=MovieSceneScriptingFloatChannel}
         ]
        }]
      }]
    }
    ButtonBacking {
     id=#2
     possessed_class=Image
     tracks [{
       type=MovieSceneWidgetMaterialTrack
       sections [{
         type=MovieSceneParameterSection
         end_frame=4
         start_frame=0
         channels [AnimateSelection {type=MovieSceneScriptingFloatChannel, keys [[0,0],[4,1]]}]
        }]
      }]
    }
    VariableSize {
     id=#5
     possessed_class=SizeBox
     tracks [{
       type=MovieScene2DTransformTrack
       sections [{
         type=MovieScene2DTransformSection
         end_frame=4
         start_frame=0
         channels [
          Translation.X {type=MovieSceneScriptingFloatChannel}
          Translation.Y {type=MovieSceneScriptingFloatChannel}
          Angle {type=MovieSceneScriptingFloatChannel}
          Scale.X {type=MovieSceneScriptingFloatChannel, keys [[0,1],[4,0.975]]}
          Scale.Y {type=MovieSceneScriptingFloatChannel, keys [[0,1],[4,0.975]]}
          Shear.X {type=MovieSceneScriptingFloatChannel}
          Shear.Y {type=MovieSceneScriptingFloatChannel}
         ]
        }]
      }]
    }
   ]}
 }
 Hover_Mobile {
  binding_count=3
  length_seconds=0.2
  object=&self:Hover_Mobile
  sequence {bindings [
    AmountColor {
     id=#1
     possessed_class=Border
     tracks [{
       type=MovieSceneColorTrack
       display_name=Content Color and Opacity
       sections [{
         type=MovieSceneColorSection
         end_frame=4
         start_frame=0
         channels [
          Color.R {type=MovieSceneScriptingFloatChannel, keys [[0,1],[1,0.8438],[2,0.5139],[3,0.1866],[4,0]]}
          Color.G {type=MovieSceneScriptingFloatChannel, thinned_from=5, keys [[0,1],[1,0.9479],[3,0.7289],[4,0.6667]]}
          Color.B {type=MovieSceneScriptingFloatChannel}
          Color.A {type=MovieSceneScriptingFloatChannel}
         ]
        }]
      }]
    }
    ButtonBacking {
     id=#2
     possessed_class=Image
     tracks [{
       type=MovieSceneWidgetMaterialTrack
       sections [{
         type=MovieSceneParameterSection
         end_frame=4
         start_frame=0
         channels [AnimateSelection {type=MovieSceneScriptingFloatChannel, keys [[0,0],[4,1]]}]
        }]
      }]
    }
    VBuckTint {
     id=#3
     possessed_class=Border
     tracks [{
       type=MovieSceneColorTrack
       display_name=Content Color and Opacity
       sections [{
         type=MovieSceneColorSection
         end_frame=4
         start_frame=0
         channels [
          Color.R {type=MovieSceneScriptingFloatChannel, keys [[0,0.675],[4,1]]}
          Color.G {type=MovieSceneScriptingFloatChannel, keys [[0,0.825],[4,1]]}
          Color.B {type=MovieSceneScriptingFloatChannel, keys [[0,0.9],[4,1]]}
          Color.A {type=MovieSceneScriptingFloatChannel, const [1,1]}
         ]
        }]
      }]
    }
   ]}
 }
]
graphs [
 EventGraph {nodes [
   Event On Hovered()
   n1 = IsUsingTouch(PlayerController=GetOwningPlayer())
   > n2 = Branch(Condition=@n1)
   PlayAnimationForward(execute=[@n7.then, @n2.else], InAnimation=GetHoverAnimation!() -> (ReturnValue:Widget Animation Object Reference), PlaybackSpeed=1.5)
   Event Construct()
   n3 = IsMobileGame()
   n4 = Get VariableSize() -> (VariableSize:Size Box Object Reference)
   > SetHeightOverride(self=@n4, InHeightOverride=Select(Option 0:double=48, Option 1:double=105, Index=@n3) -> (ReturnValue:double))
   > SetMinDesiredWidth!(execute, self:Size Box Object Reference=@n4, InMinDesiredWidth:float=Select(Option 0=88, Option 1=120, Index=@n3)) -> (then)
   n5 = AND Boolean(
    A=IsMobileGame()
    B=Get bBacchusFrontendEnabled!(self:Fort Runtime Options Object Reference=GetRuntimeOptions()) -> (bBacchusFrontendEnabled:bool)
   )
   > SetBrushFromMaterial(
    self=Get ButtonBacking() -> (ButtonBacking:Image Object Reference)
    Material=SelectMaterialInstance!(
     A:Material Instance Object Reference=/FortUILibrary/Foundation/Buttons/Materials/TopBar/M_UI_TopBar_PanelButtons_Bacchus
     B:Material Instance Object Reference=/FortUILibrary/Foundation/Buttons/Materials/TopBar/M_UI_TopBar_PanelButtons_Hamburger
     bSelectA:bool=@n5
    ) -> (ReturnValue:Material Instance Object Reference)
   )
   > SetContentColorAndOpacity!(
    execute
    self:Border Object Reference=Get AmountColor() -> (AmountColor:Border Object Reference)
    InContentColorAndOpacity:Linear Color Structure=SelectColor(A=(R=1,G=1,B=1,A=1), B=(R=0,G=0.6667,B=1,A=1), bPickA=@n5)
   ) -> (then)
   Event On Unhovered()
   > n6 = Branch(Condition=@n1)
   PlayAnimationReverse(execute=[@n8.then, @n6.else], InAnimation=GetHoverAnimation(), PlaybackSpeed=2)
   Event On Pressed() {is=title K2Node_Event}
   > n7 = Branch(Condition=@n1)
   Event On Released() {is=title K2Node_Event}
   > n8 = Branch(Condition=@n1)
  ]}
 GetHoverAnimation {nodes [
   GetHoverAnimation() {is=title K2Node_FunctionEntry}
   n1 = IsMobileGame()
   n2 = Get bBacchusFrontendEnabled(self=GetRuntimeOptions())
   > Return Node!(
    execute
    ReturnValue:Widget Animation Object Reference=Select(
     Option 0:Widget Animation Object Reference=Get Hover() -> (Hover:Widget Animation Object Reference)
     Option 1:Widget Animation Object Reference=Get Hover_Mobile() -> (Hover_Mobile:Widget Animation Object Reference)
     Index=AND Boolean(A=@n1, B=@n2)
    ) -> (ReturnValue:Widget Animation Object Reference)
   )
  ]}
]
variables []
widgets {
 info {inheritedWidgetCount=2, parentClass=/Script/FortniteUI.FortMTXButton, rootWidgetClass=&Overlay, widgetCount=10}
 widgets [
  Overlay_0 {widgetClassPath=&Overlay, widget=Overlay_0}
  VariableSize {
   widgetClassPath=&SizeBox
   bIsVariable=true
   parent=Overlay_0
   widget=VariableSize
   properties {bOverride_MinDesiredHeight=true, bOverride_MinDesiredWidth=true, heightOverride=48, minDesiredHeight=48, minDesiredWidth=128}
   slot {class=OverlaySlot, properties {horizontalAlignment=HAlign_Fill, verticalAlignment=VAlign_Fill}}
  }
  ButtonBacking {
   widgetClassPath=&Image
   bIsVariable=true
   parent=VariableSize
   widget=ButtonBacking
   properties {brush {
     bIsDynamicallyLoaded=false
     drawAs=Image
     imageType=NoImage
     mirroring=NoMirror
     resourceName=None
     resourceObject=/FortUILibrary/Foundation/Buttons/Materials/TopBar/M_UI_TopBar_PanelButtons_MTX
     tiling=NoTile
     imageSize=<32 32>
     margin=<ltrb 0 0 0 0>
     outlineSettings {
      bUseBrushTransparency=false
      roundingType=HalfHeightRadius
      width=0
      color {colorUseRule=UseColor_Specified, specifiedColor=<rgba 0 0 0 0>}
      cornerRadii=<0 0 0 0>
     }
     tintColor {colorUseRule=UseColor_Specified, specifiedColor=<rgba 1 1 1 1>}
     uVRegion {bIsValid=false, max=<0 0>, min=<0 0>}
    }}
   slot {class=SizeBoxSlot}
  }
  HorizontalBox_0 {
   widgetClassPath=&HorizontalBox
   parent=Overlay_0
   widget=HorizontalBox_0
   slot {class=OverlaySlot, properties {horizontalAlignment=HAlign_Fill, verticalAlignment=VAlign_Fill}}
  }
  SizeBox_1 {
   widgetClassPath=&SizeBox
   parent=HorizontalBox_0
   widget=SizeBox_1
   properties {bOverride_HeightOverride=true, bOverride_WidthOverride=true, heightOverride=30, widthOverride=30}
   slot {class=HorizontalBoxSlot, properties {verticalAlignment=VAlign_Center, padding=<ltrb 16 0 8 0>}}
  }
  VBuckTint {
   widgetClassPath=&Border
   parent=SizeBox_1
   widget=VBuckTint
   properties {
    background {
     bIsDynamicallyLoaded=false
     drawAs=NoDrawType
     imageType=NoImage
     mirroring=NoMirror
     resourceName=None
     resourceObject=null
     tiling=NoTile
     imageSize=<32 32>
     margin=<ltrb 0 0 0 0>
     outlineSettings {
      bUseBrushTransparency=false
      roundingType=HalfHeightRadius
      width=0
      color {colorUseRule=UseColor_Specified, specifiedColor=<rgba 0 0 0 0>}
      cornerRadii=<0 0 0 0>
     }
     tintColor {colorUseRule=UseColor_Specified, specifiedColor=<rgba 1 1 1 1>}
     uVRegion {bIsValid=false, max=<0 0>, min=<0 0>}
    }
    contentColorAndOpacity=<rgba 0.675 0.825 0.9 1>
    padding=<ltrb 0 0 0 0>
   }
   slot {class=SizeBoxSlot}
  }
  Image_Vbucks {
   widgetClassPath=&Image
   bIsVariable=true
   parent=VBuckTint
   widget=Image_Vbucks
   properties {brush {
     bIsDynamicallyLoaded=false
     drawAs=Image
     imageType=FullColor
     mirroring=NoMirror
     resourceName=None
     resourceObject=/Game/UI/Foundation/Textures/Icons/Currency/T-VBuck-64
     tiling=NoTile
     imageSize=<32 32>
     margin=<ltrb 0 0 0 0>
     outlineSettings {
      bUseBrushTransparency=false
      roundingType=HalfHeightRadius
      width=0
      color {colorUseRule=UseColor_Specified, specifiedColor=<rgba 0 0 0 0>}
      cornerRadii=<0 0 0 0>
     }
     tintColor {colorUseRule=UseColor_Specified, specifiedColor=<rgba 1 1 1 1>}
     uVRegion {bIsValid=false, max=<0 0>, min=<0 0>}
    }}
   slot {class=BorderSlot, properties {padding=<ltrb 0 0 0 0>}}
  }
  AmountColor {
   widgetClassPath=&Border
   bIsVariable=true
   parent=HorizontalBox_0
   widget=AmountColor
   properties {
    background {
     bIsDynamicallyLoaded=false
     drawAs=NoDrawType
     imageType=NoImage
     mirroring=NoMirror
     resourceName=None
     resourceObject=null
     tiling=NoTile
     imageSize=<32 32>
     margin=<ltrb 0 0 0 0>
     outlineSettings {
      bUseBrushTransparency=false
      roundingType=HalfHeightRadius
      width=0
      color {colorUseRule=UseColor_Specified, specifiedColor=<rgba 0 0 0 0>}
      cornerRadii=<0 0 0 0>
     }
     tintColor {colorUseRule=UseColor_Specified, specifiedColor=<rgba 1 1 1 1>}
     uVRegion {bIsValid=false, max=<0 0>, min=<0 0>}
    }
    contentColorAndOpacity=<rgba 0 0.6667 1 1>
    padding=<ltrb 0 5 0 5>
    renderTransform {angle=0, scale=<1 1>, shear=<0 0>, translation=<0 2>}
   }
   slot {class=HorizontalBoxSlot, properties {verticalAlignment=VAlign_Center, padding=<ltrb 0 0 20 2>}}
  }
  Text_Amount {
   widgetClassPath=/Script/CommonUI.CommonNumericTextBlock
   bInherited=true
   bIsVariable=true
   parent=AmountColor
   widget=Text_Amount
   properties {
    currentNumericValue=123456
    mobileFontSizeMultiplier=1.5
    style=/FortUILibrary/Foundation/Buttons/ButtonStyles/Athena/TextStyle_BurbSmBk_20_White.TextStyle_BurbSmBk_20_White_C
    text="123,456"
    font {
     bForceMonospaced=false
     bMaterialIsStencil=false
     fontMaterial=null
     fontObject=&BurbankSmall
     letterSpacing=0
     monospacedWidth=1
     size=20
     skewAmount=0
     typefaceFontName=Black
     outlineSettings {
      bApplyOutlineToDropShadows=false
      bMiteredCorners=false
      bSeparateFillAlpha=false
      outlineMaterial=null
      outlineSize=0
      outlineColor=<rgba 0 0 0 1>
     }
    }
    shadowOffset=<0 0>
   }
   slot {class=BorderSlot, properties {padding=<ltrb 0 5 0 5>}}
  }
  InputActionWidget {widgetClassPath=&CommonActionWidget, bInherited=true, widget=null}
 ]
}
