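# Top-bar onboarding tooltip (FortSidebarOnboardTooltipWidget) that announces the upgraded Career page through a modal UIKit coach mark.
# PreConstruct styles the coach mark's hidden other-action button and sets top padding for mobile or non-mobile.
# The coach mark activates only for gamepad input, checked on show and on input method changes.
# Dismiss and OnEndTooltip both run HandleOtherActionClicked, report close finished and deactivate, which unbinds events and deactivates the coach mark.
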WidgetBlueprint /Game/UI/Frontend/TopBar/WBP_CareerV3_CoachMark
graphs [EventGraph {nodes [
   n1 = On Input Method Changed!() -> (OutputDelegate, then, bNewInputType:ECommonInputType Enum) {is=title K2Node_CustomEvent}
   > Branch(Condition="Equal (Enum)"(A=@n1.bNewInputType, B=Gamepad))
   > ActivateWidget(self=Get WBP_UIKit_CoachMark() -> (WBP_UIKit_CoachMark:WBP UIKit Coach Mark Object Reference))
   n2 = OnDismissButtonClicked!() -> (OutputDelegate, then, Button:Common Button Base Object Reference) {is=title K2Node_CustomEvent}
   HandleOtherActionClicked(execute=[@n2.then, @n3.then], self:Fort Sidebar Onboard Tooltip Widget Object Reference)
   > NotifyCloseAnimationFinished(self:Fort Sidebar Onboard Tooltip Widget Object Reference)
   > DeactivateWidget()
   Event PreConstruct()
   > SetButtonType!(
    execute
    self:WBP UIKit Button Regular Object Reference=Get OtherActionButton!(self:WBP UIKit Coach Mark Base Object Reference=Get WBP_UIKit_CoachMark()) -> (OtherActionButton:WBP UIKit Button Regular Object Reference)
    ButtonType:E_UI_CTAButtonType Enum=NewEnumerator1
   ) -> (then)
   > SetPadding!(
    execute
    self:Size Box Slot Object Reference=SlotAsSizeBoxSlot!(Widget:Widget Object Reference=Get WBP_UIKit_CoachMark()) -> (ReturnValue:Size Box Slot Object Reference)
    InPadding_Left:float=0
    InPadding_Top:float=Select(Option 0:float=70, Option 1:float=145, Index=IsMobileGame()) -> (ReturnValue:float)
    InPadding_Right:float=0
    InPadding_Bottom:float=0
   ) -> (then)
   n3 = Event OnEndTooltip() {is=title K2Node_Event}
   Event OnShowTooltip() {is=title K2Node_Event}
   n4 = Get CommonInputSubsystem()
   > Branch(Condition="Equal (Enum)"(A=GetCurrentInputType(self=@n4), B=Gamepad))
   > ActivateWidget(self=Get WBP_UIKit_CoachMark())
   Event Construct()
   n5 = Get DismissButton!(self:WBP UIKit Coach Mark Base Object Reference=Get WBP_UIKit_CoachMark()) -> (DismissButton:WBP UIKit Button Regular Object Reference)
   > Assign On Clicked!(execute:Exec, self:Common Button Base Object Reference=@n5, Delegate:"Delegate (by ref)"=@n2) -> (then:Exec) {is=title K2Node_AssignDelegate}
   n6 = Get CommonInputSubsystem()
   > Bind Event to On Input Method Changed(self=@n6, Delegate=@n1)
   Event On Deactivated()
   > Unbind all Events from On Clicked!(execute:Exec, self:Common Button Base Object Reference=@n5) -> (then:Exec) {is=title K2Node_ClearDelegate}
   > Unbind Event from On Input Method Changed(self=@n6, Delegate=@n1)
   > DeactivateWidget(self=Get WBP_UIKit_CoachMark())
  ]}]
variables []
widgets {
 info {
  inheritedWidgetCount=2
  parentClass=/Script/FortniteUI.FortSidebarOnboardTooltipWidget
  rootWidgetClass=&CommonVisualAttachment
  widgetCount=4
 }
 widgets [
  CommonVisualAttachment {
   widgetClassPath=&CommonVisualAttachment
   bIsVariable=true
   widget=CommonVisualAttachment
   properties {contentAnchor=<0.11 0>}
  }
  WBP_UIKit_CoachMark {
   widgetClassPath=/FortUILibrary/UIKit/CoachMark/WBP_UIKit_CoachMark
   bIsVariable=true
   parent=CommonVisualAttachment
   widget=WBP_UIKit_CoachMark
   properties {
    arrowPosition=Top_Left
    bIsFocusable=true
    bIsModal=true
    bSupportsActivationFocus=false
    navigation=&self:WidgetTree.WBP_UIKit_CoachMark.WidgetNavigation_0
    shouldShowOtherActionButton=false
    textDescription="<text color=\"C8C8C8\">Check out the new look and feel of the Career page.</>"
    textHeader=Career just got an upgrade
    textOtherActionButton="go there!"
    visibility=SelfHitTestInvisible
    in Decorator Classes []
    input Action Dismiss {dataTable=/Game/UI/Frontend/Lobby/Habanero/HabaneroData/DT_HabaneroUIInputActions, rowName=Back_Controller}
    input Action Other {dataTable=&InputActionDataTable, rowName=GenericRightStickClick}
   }
   slot {class=SizeBoxSlot, properties {horizontalAlignment=HAlign_Left, verticalAlignment=VAlign_Top}}
  }
  Dismiss_Button {widgetClassPath=&CommonButtonLegacy, bInherited=true, widget=null}
  OtherAction_Button {widgetClassPath=&CommonButtonLegacy, bInherited=true, widget=null}
 ]
}
