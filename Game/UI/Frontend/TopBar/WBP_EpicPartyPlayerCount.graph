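# Top-bar badge widget that shows the Epic party player count beside a headphones icon.
# PreConstruct tints the badge background green through a dynamic material parameter.
# Construct caches the Fort.UseEpicParties console variable in isEpicParty.
# SetPartyCountAndVisibility shows the widget as hit-test invisible and writes the count only when it exceeds one and Epic parties are enabled, otherwise collapsing it.
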WidgetBlueprint /Game/UI/Frontend/TopBar/WBP_EpicPartyPlayerCount
graphs [
 EventGraph {nodes [
   Event PreConstruct()
   > n1 = GetDynamicMaterial(self=Get Background() -> (Background:Image Object Reference))
   > SetVectorParameterValue(self=@n1, ParameterName=Color, Value=(R=0.0369,G=1,B=0.2232,A=1))
   Event Construct()
   > n2 = GetConsoleVariableBoolValue!(execute, VariableName:String=Fort.UseEpicParties) -> (then, ReturnValue:bool)
   > Set isEpicParty(isEpicParty:bool=@n2) -> (Output_Get:bool)
  ]}
 SetPartyCountAndVisibility {nodes [
   n1 = SetPartyCountAndVisibility!() -> (then, Count:int) {is=title K2Node_FunctionEntry}
   > n2 = Branch(Condition=AND Boolean(A="integer > integer"(A=@n1, B=1), B=Get isEpicParty() -> (isEpicParty:bool)))
   > SetVisibility(InVisibility=HitTestInvisible)
   > "SetText (Text)"(self=Get Text!() -> (Text:Common Text Object Reference), InText="To Text (Integer)"(Value=Get Count() -> (Count:int=0)))
   SetVisibility(execute=@n2.else, InVisibility=Collapsed)
  ]}
]
variables [isEpicParty {type=boolean}]
widgets {
 info {inheritedWidgetCount=0, parentClass=&UserWidget, rootWidgetClass=&GridPanel, widgetCount=6}
 widgets [
  Root {widgetClassPath=&GridPanel, widget=Root, properties {columnFill [1], rowFill [1]}}
  Outline {
   widgetClassPath=&Image
   bIsVariable=true
   parent=Root
   widget=Outline
   properties {
    visibility=SelfHitTestInvisible
    brush {
     bIsDynamicallyLoaded=false
     drawAs=Image
     imageType=FullColor
     mirroring=NoMirror
     resourceName=None
     resourceObject=/FortUILibrary/Systems/Badge/MaterialInstances/MI_UIKit_BadgeBackground
     tiling=NoTile
     imageSize=<16 16>
     margin=<ltrb 0 0 0 0>
     outlineSettings {
      bUseBrushTransparency=false
      roundingType=HalfHeightRadius
      width=0
      color {colorUseRule=UseColor_Specified, specifiedColor=<rgba 0 0 0 0>}
      cornerRadii=<0 0 0 0>
     }
     tintColor {colorUseRule=UseColor_Specified, specifiedColor=<rgba 1 1 1 0.35>}
     uVRegion {bIsValid=false, max=<0 0>, min=<0 0>}
    }
   }
   slot {class=GridSlot, properties {padding=<ltrb -4 -4 -4 -4>}}
  }
  Background {
   widgetClassPath=&Image
   bIsVariable=true
   parent=Root
   widget=Background
   properties {
    visibility=SelfHitTestInvisible
    brush {
     bIsDynamicallyLoaded=false
     drawAs=Image
     imageType=FullColor
     mirroring=NoMirror
     resourceName=None
     resourceObject=&MI_UIKit_BadgeBackground
     tiling=NoTile
     imageSize=<16 16>
     margin=<ltrb 0 0 0 0>
     outlineSettings {
      bUseBrushTransparency=false
      roundingType=HalfHeightRadius
      width=0
      color {colorUseRule=UseColor_Specified, specifiedColor=<rgba 0 0 0 0>}
      cornerRadii=<0 0 0 0>
     }
     tintColor {colorUseRule=UseColor_Specified, specifiedColor=<rgba 1 1 1 1>}
     uVRegion {bIsValid=false, max=<0 0>, min=<0 0>}
    }
   }
   slot {class=GridSlot}
  }
  Content {
   widgetClassPath=&GridPanel
   bIsVariable=true
   parent=Root
   widget=Content
   properties {columnFill [0,1], rowFill [1]}
   slot {class=GridSlot, properties {padding=<ltrb 4 0 6 0>}}
  }
  Text {
   widgetClassPath=&CommonTextBlock
   bIsVariable=true
   parent=Content
   widget=Text
   properties {
    applyLineHeightToBottomLine=false
    bAutoCollapseWithEmptyText=true
    scrollStyle=&TextScrollStyle-Base_C
    text=0
    colorAndOpacity {colorUseRule=UseColor_Specified, specifiedColor=<rgba 0.0086 0.0086 0.0116 0.949>}
    font {
     bForceMonospaced=false
     bMaterialIsStencil=false
     fontMaterial=null
     fontObject=&HeadingNow
     letterSpacing=0
     monospacedWidth=1
     size=13.5
     skewAmount=0
     typefaceFontName=75 Medium
     outlineSettings {
      bApplyOutlineToDropShadows=false
      bMiteredCorners=false
      bSeparateFillAlpha=false
      outlineMaterial=null
      outlineSize=0
      outlineColor=<rgba 0 0 0 1>
     }
    }
    shadowOffset=<0 0>
   }
   slot {class=GridSlot, properties {column=1, horizontalAlignment=HAlign_Left, layer=1, verticalAlignment=VAlign_Center, padding=<ltrb 4 0 0 0>}}
  }
  Icon {
   widgetClassPath=&Image
   bIsVariable=true
   parent=Content
   widget=Icon
   properties {
    brush {
     bIsDynamicallyLoaded=false
     drawAs=Image
     imageType=FullColor
     mirroring=NoMirror
     resourceName=None
     resourceObject=/FortUILibrary/Textures/MaterialInstances/MI_UI_Icon_UIKit_Headphones_MSDF
     tiling=NoTile
     imageSize=<24 24>
     margin=<ltrb 0 0 0 0>
     outlineSettings {
      bUseBrushTransparency=false
      roundingType=HalfHeightRadius
      width=0
      color {colorUseRule=UseColor_Specified, specifiedColor=<rgba 0 0 0 0>}
      cornerRadii=<0 0 0 0>
     }
     tintColor {colorUseRule=UseColor_Specified, specifiedColor=<rgba 1 1 1 1>}
     uVRegion {bIsValid=false, max=<0 0>, min=<0 0>}
    }
    colorAndOpacity=<rgba 0.0086 0.0086 0.0116 0.949>
   }
   slot {class=GridSlot}
  }
 ]
}
