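# Frontend top-bar onboarding coach mark that tells players they can search for creators and favorite them.
# When shown, it checks the Follow-a-Creator and search-tab runtime options, then activates the modal coach mark for gamepad input, also doing so on switches to gamepad.
# "Got it", tooltip end, or mouse-and-keyboard focus loss notifies close and deactivates.
# PreConstruct pads differently on mobile; a debug PrintString remains after show-time activation.
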WidgetBlueprint /Game/UI/Frontend/TopBar/WBP_SearchForCreators_CoachMark
graphs [EventGraph {nodes [
   "On Widget Activated (WBP_UIKit_CoachMark)"() {is=title K2Node_ComponentBoundEvent}
   n1 = Get WBP_UIKit_CoachMark() -> (WBP_UIKit_CoachMark:WBP UIKit Coach Mark Object Reference)
   > n2 = Branch(Condition=HasUserFocus!(
     self:Widget Object Reference=Get OtherActionButton!(self:WBP UIKit Coach Mark Base Object Reference=@n1) -> (OtherActionButton:WBP UIKit Button Regular Object Reference)
     PlayerController:Player Controller Object Reference=GetOwningPlayer()
    ) -> (ReturnValue:bool))
   DeactivateWidget(execute=@n2.else)
   n3 = OnOtherButtonClicked!() -> (OutputDelegate, then, Button:Common Button Base Object Reference) {is=title K2Node_CustomEvent}
   > n4 = HandleOtherActionClicked(self:Fort Sidebar Onboard Tooltip Widget Object Reference)
   NotifyCloseAnimationFinished(execute=[@n4.then, @n9.then, @n12.then], self:Fort Sidebar Onboard Tooltip Widget Object Reference)
   > DeactivateWidget()
   n5 = On Input Method Changed!() -> (OutputDelegate, then, bNewInputType:ECommonInputType Enum) {is=title K2Node_CustomEvent}
   > Branch(Condition="Equal (Enum)"(A=@n5.bNewInputType, B=Gamepad))
   > ActivateWidget(self=Get WBP_UIKit_CoachMark())
   Event PreConstruct()
   > SetButtonType!(
    execute
    self:WBP UIKit Button Regular Object Reference=Get OtherActionButton(self=Get WBP_UIKit_CoachMark())
    ButtonType:E_UI_CTAButtonType Enum=NewEnumerator1
   ) -> (then)
   > SetPadding!(
    execute
    self:Size Box Slot Object Reference=SlotAsSizeBoxSlot!(Widget:Widget Object Reference=Get WBP_UIKit_CoachMark()) -> (ReturnValue:Size Box Slot Object Reference)
    InPadding_Left:float=0
    InPadding_Top:float=Select(Option 0:float=70, Option 1:float=145, Index=IsMobileGame()) -> (ReturnValue:float)
    InPadding_Right:float=0
    InPadding_Bottom:float=0
   ) -> (then)
   Event OnShowTooltip() {is=title K2Node_Event}
   n6 = GetRuntimeOptions()
   n7 = Self-Reference()
   > Branch(Condition=AND Boolean(
     A=GetIsFollowACreatorPageEnabled!(self:Fort Runtime Options Object Reference=@n6) -> (ReturnValue:bool)
     B=GetIsSearchTabInTopNavigationBarEnabled!(self:Fort Runtime Options Object Reference=@n6, ContextWidget:Widget Object Reference=@n7) -> (ReturnValue:bool)
    ))
   n8 = Get CommonInputSubsystem()
   > Branch(Condition="Equal (Enum)"(A=GetCurrentInputType(self=@n8), B=Gamepad))
   > ActivateWidget(self=Get WBP_UIKit_CoachMark())
   > PrintString!(
    execute
    InString=WBP_Tooltip_SearchForCreators is created but mb its invisible
    bPrintToScreen
    bPrintToLog
    TextColor=(R=0,G=1,B=0.0363,A=1)
    Duration=5
    Key
   ) -> (then)
   n9 = Event OnEndTooltip() {is=title K2Node_Event}
   Event OnRemovedFromFocusPath!() -> (OutputDelegate, then, InFocusEvent:Focus Event Structure) {is=title K2Node_Event}
   n10 = Get WBP_UIKit_CoachMark()
   n11 = GetCurrentInputType(self=Get CommonInputSubsystem())
   > n12 = Branch(Condition=AND Boolean(A=IsActivated!(self:Common Activatable Widget Object Reference=@n10) -> (ReturnValue:bool), B="Equal (Enum)"(A=@n11, B=MouseAndKeyboard)))
   Event Construct()
   n13 = Get OtherActionButton(self=Get WBP_UIKit_CoachMark())
   > Bind Event to On Clicked(self:Common Button Base Object Reference=@n13, Delegate=@n3) {is=title K2Node_AddDelegate}
   n14 = Get CommonInputSubsystem()
   > Bind Event to On Input Method Changed(self=@n14, Delegate=@n5)
   Event On Deactivated()
   > Unbind all Events from On Clicked!(execute:Exec, self:Common Button Base Object Reference=@n13) -> (then:Exec) {is=title K2Node_ClearDelegate}
   > Unbind all Events from On Input Method Changed!(execute:Exec, self:Common Input Object Reference=@n14) -> (then:Exec) {is=title K2Node_ClearDelegate}
   > DeactivateWidget(self=Get WBP_UIKit_CoachMark())
  ]}]
variables []
widgets {
 info {
  inheritedWidgetCount=2
  parentClass=/Script/FortniteUI.FortSidebarOnboardTooltipWidget
  rootWidgetClass=&CommonVisualAttachment
  widgetCount=4
 }
 widgets [
  CommonVisualAttachment {
   widgetClassPath=&CommonVisualAttachment
   bIsVariable=true
   widget=CommonVisualAttachment
   properties {contentAnchor=<0.11 0>}
  }
  WBP_UIKit_CoachMark {
   widgetClassPath=/FortUILibrary/UIKit/CoachMark/WBP_UIKit_CoachMark
   bIsVariable=true
   parent=CommonVisualAttachment
   widget=WBP_UIKit_CoachMark
   properties {
    arrowPosition=Top_Left
    bIsFocusable=true
    bIsModal=true
    navigation=&self:WidgetTree.WBP_UIKit_CoachMark.WidgetNavigation_0
    shouldShowDismissButton=false
    textDescription="<text color=\"C8C8C8\">Find creators and add them to your favorites</>"
    textHeader=Search for creators
    textOtherActionButton=Got it
    visibility=SelfHitTestInvisible
    input Action Other {dataTable=&InputActionDataTable, rowName=GenericForward}
   }
   slot {class=SizeBoxSlot, properties {horizontalAlignment=HAlign_Left, verticalAlignment=VAlign_Top}}
  }
  Dismiss_Button {widgetClassPath=&CommonButtonLegacy, bInherited=true, widget=null}
  OtherAction_Button {widgetClassPath=&CommonButtonLegacy, bInherited=true, widget=null}
 ]
}
