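# Sidebar avatar button for the frontend top bar, built on the generic UIKit button, showing the local player's avatar, party count and ready status.
# On Construct it binds the avatar to the local player and caches its view model; hover drives the outline block's focus transition.
# Handlers update the ready icon and player-number label, which appears, shifting the party-count anchor, only with multiple splitscreen players.
# Compact UI adjusts sizing and icon layout in PreConstruct.
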WidgetBlueprint /Game/UI/Frontend/TopBar/WBP_SidebarAvatarButton
graphs [
 EventGraph {nodes [
   n1 = Event PreConstruct()
   > Parent: PreConstruct(IsDesignTime=@n1.IsDesignTime)
   > SetCompactUIStyle() {is=title K2Node_CallFunction}
   Event Construct()
   > Parent: Construct()
   n2 = Get WBP_PlayerAvatar() -> (WBP_PlayerAvatar:WBP Player Avatar Object Reference)
   > Branch(Condition=IsValid(Object=@n2))
   > RepresentLocalPlayer(self:Fort Player Avatar Object Reference=@n2)
   > Set FortPlayerAvatarViewModel(FortPlayerAvatarViewModel:Fort Player Avatar Viewmodel Object Reference=Get FortPlayerAvatarViewModel!(self:WBP Player Avatar Object Reference=@n2) -> (FortPlayerAvatarViewModel:Fort Player Avatar Viewmodel Object Reference)) -> (Output_Get:Fort Player Avatar Viewmodel Object Reference)
   Event On Hovered()
   > Parent: On Hovered() {is=title K2Node_CallParentFunction}
   n3 = Get Block_Outline() -> (Block_Outline:WBP UIKit Block Outline Object Reference)
   > RequestTransition!(execute, self:UIKit Block Object Reference=@n3, State:EUIKitBlockVisualState Enum=Focused) -> (then)
   Event On Unhovered()
   > Parent: On Unhovered() {is=title K2Node_CallParentFunction}
   > RequestTransition(self=@n3, State=Unfocused)
  ]}
 HandleLocalPlayerCountChanged {nodes [
   n1 = HandleLocalPlayerCountChanged!() -> (then, NumLocalPlayers:int) {is=title K2Node_FunctionEntry}
   n2 = AND Boolean(
    A="integer > integer"(A=@n1, B=1)
    B=IsSplitscreenV2Enabled!(self:Fort Runtime Options Object Reference=GetRuntimeOptions()) -> (ReturnValue:bool)
   )
   > SetVisibility(
    self=Get Overlay_PlayerNum() -> (Overlay_PlayerNum:Overlay Object Reference)
    InVisibility=Select(Option 0=Collapsed, Option 1=SelfHitTestInvisible, Index=@n2)
   )
   > Set ContentAnchor!(
    execute
    ContentAnchor:Vector 2D Structure=Select(Option 0:Vector 2D Structure=(X=0.5,Y=0.9), Option 1:Vector 2D Structure=(X=0,Y=0.45), Index=@n2) -> (ReturnValue:Vector 2D Structure)
    self:Common Visual Attachment Object Reference=Get CommonVisualAttachment_PartyCount() -> (CommonVisualAttachment_PartyCount:Common Visual Attachment Object Reference)
   ) -> (then, Output_Get:Vector 2D Structure)
  ]}
 HandleLocalPlayerNumSet {nodes [
   n1 = HandleLocalPlayerNumSet!() -> (then, LocalPlayerNum:int) {is=title K2Node_FunctionEntry}
   n2 = int + int(A=@n1)
   > "SetText (Text)"(
    self=Get Text_PlayerNum() -> (Text_PlayerNum:Common Text Object Reference)
    InText=Format Text!(Format=Get PlayerFormatString() -> (PlayerFormatString:Text), PlayerNum:Text="To Text (Integer)"(Value=@n2)) -> (Result)
   )
  ]}
 HandleReadyStatusChanged {nodes [
   n1 = HandleReadyStatusChanged!() -> (then, ReadyStatus:EGameReadiness Enum=NotReady) {is=title K2Node_FunctionEntry}
   > UpdateStatusIcon!(
    execute
    self:WBP Nameplate Matchmaking Status Icon Object Reference=Get WBP_Nameplate_MatchmakingStatusIcon_90() -> (WBP_Nameplate_MatchmakingStatusIcon_90:WBP Nameplate Matchmaking Status Icon Object Reference)
    CurrentMatchmakingStatus:EPedestalNameplateMatchmakingStatus Enum=Select(
     Option 0:EPedestalNameplateMatchmakingStatus Enum=NotReady
     Option 1:EPedestalNameplateMatchmakingStatus Enum=Ready
     Index="Equal (Enum)"(A:EGameReadiness Enum=@n1, B:EGameReadiness Enum=Ready)
    ) -> (ReturnValue:EPedestalNameplateMatchmakingStatus Enum)
   ) -> (then)
  ]}
 SetCompactUIStyle {nodes [
   SetCompactUIStyle() {is=title K2Node_FunctionEntry}
   > Branch(Condition=IsUsingCompactUI())
   n1 = Get SB_mobileSizeGuard() -> (SB_mobileSizeGuard:Size Box Object Reference)
   > SetWidthOverride(self=@n1, InWidthOverride=102)
   > SetHeightOverride(self=@n1, InHeightOverride=102)
   > Set Show Sharing Activity Visuals!(execute, Show Sharing Activity Visuals:bool=true, self:WBP Player Avatar Object Reference=Get WBP_PlayerAvatar()) -> (then, Output_Get:bool)
   n2 = Get WBP_Nameplate_MatchmakingStatusIcon_90()
   > SetRenderScale!(execute, self:Widget Object Reference=@n2, Scale:Vector 2D Structure=(X=1.5,Y=1.5)) -> (then)
   > SetPadding!(
    execute
    self:Overlay Slot Object Reference=SlotAsOverlaySlot(Widget=@n2)
    InPadding_Left:float=0
    InPadding_Top:float=0
    InPadding_Right:float=8
    InPadding_Bottom:float=8
   ) -> (then)
  ]}
]
variables []
widgets {
 info {
  inheritedWidgetCount=0
  namedSlotCount=2
  parentClass=/FortUILibrary/UIKit/Buttons/WBP_UIKit_Button_Generic.WBP_UIKit_Button_Generic_C
  rootWidgetClass=null
  widgetCount=15
 }
 widgets [
  GridPanel_30 {widgetClassPath=&GridPanel, widget=GridPanel_30, properties {columnFill [1], rowFill [1]}}
  SB_mobileSizeGuard {
   widgetClassPath=&SizeBox
   bIsVariable=true
   parent=GridPanel_30
   widget=SB_mobileSizeGuard
   properties {heightOverride=102, widthOverride=102}
   slot {class=GridSlot, properties {horizontalAlignment=HAlign_Center, layer=3, verticalAlignment=VAlign_Center}}
  }
  HB_Content {
   widgetClassPath=&HorizontalBox
   parent=SB_mobileSizeGuard
   widget=HB_Content
   slot {class=SizeBoxSlot, properties {padding=<ltrb 5 5 5 5>}}
  }
  Overlay_Avatar {
   widgetClassPath=&Overlay
   parent=HB_Content
   widget=Overlay_Avatar
   properties {pixelSnapping=Disabled}
   slot {class=HorizontalBoxSlot, properties {size {sizeRule=Fill, value=1}}}
  }
  WBP_PlayerAvatar {
   widgetClassPath=/Game/UI/Foundation/Profile/WBP_PlayerAvatar
   bIsVariable=true
   parent=Overlay_Avatar
   widget=WBP_PlayerAvatar
   properties {pixelSnapping=Disabled, presenceIconScale=0.35, show Sharing Activity Visuals=true, avatarSize=<50 50>, presenceIconOffset=<-0.8125 -0.8125>}
   slot {class=OverlaySlot, properties {horizontalAlignment=HAlign_Fill, verticalAlignment=VAlign_Fill}}
  }
  CommonVisualAttachment_PartyCount {
   widgetClassPath=&CommonVisualAttachment
   bIsVariable=true
   parent=Overlay_Avatar
   widget=CommonVisualAttachment_PartyCount
   properties {contentAnchor=<0.5 0.9>}
   slot {class=OverlaySlot, properties {horizontalAlignment=HAlign_Right, verticalAlignment=VAlign_Bottom}}
  }
  WBP_EpicPartyPlayerCount {
   widgetClassPath=&/WBP_EpicPartyPlayerCount
   bIsVariable=true
   parent=CommonVisualAttachment_PartyCount
   widget=WBP_EpicPartyPlayerCount
   slot {class=SizeBoxSlot}
  }
  WBP_Nameplate_MatchmakingStatusIcon_90 {
   widgetClassPath=/Game/UI/Frontend/Lobby/Nameplates/WBP_Nameplate_MatchmakingStatusIcon
   bIsVariable=true
   parent=Overlay_Avatar
   widget=WBP_Nameplate_MatchmakingStatusIcon_90
   slot {class=OverlaySlot, properties {horizontalAlignment=HAlign_Right, verticalAlignment=VAlign_Bottom, padding=<ltrb 0 0 -4 -4>}}
  }
  Overlay_PlayerNum {
   widgetClassPath=&Overlay
   bIsVariable=true
   parent=HB_Content
   widget=Overlay_PlayerNum
   properties {visibility=Collapsed}
   slot {class=HorizontalBoxSlot, properties {padding=<ltrb 8 0 11 0>}}
  }
  Text_PlayerNum {
   widgetClassPath=&CommonTextBlock
   bIsVariable=true
   parent=Overlay_PlayerNum
   widget=Text_PlayerNum
   properties {
    applyLineHeightToBottomLine=false
    bIsScrollingEnabled=false
    justification=Center
    text=P1
    font {
     bForceMonospaced=false
     bMaterialIsStencil=false
     fontMaterial=null
     fontObject=&HeadingNow
     letterSpacing=0
     monospacedWidth=1
     size=15
     skewAmount=0
     typefaceFontName=74 Regular
     outlineSettings {
      bApplyOutlineToDropShadows=false
      bMiteredCorners=false
      bSeparateFillAlpha=false
      outlineMaterial=null
      outlineSize=0
      outlineColor=<rgba 0 0 0 1>
     }
    }
    shadowOffset=<0 0>
   }
   slot {class=OverlaySlot, properties {horizontalAlignment=HAlign_Center, verticalAlignment=VAlign_Center}}
  }
  Block_Outline {
   widgetClassPath=/FortUILibrary/UIKit/Buttons/Blocks/WBP_UIKit_Block_Outline
   bIsVariable=true
   parent=GridPanel_30
   widget=Block_Outline
   properties {bottomLeftCornerRadius=0.5, bottomRightCornerRadius=0.5, topLeftCornerRadius=0.5, topRightCornerRadius=0.5}
   slot {class=GridSlot, properties {layer=2}}
  }
  WBP_UIKit_StatusIndicator {
   widgetClassPath=/FortUILibrary/UIKit/StatusIndicator/WBP_UIKit_StatusIndicator
   bIsVariable=true
   parent=GridPanel_30
   widget=WBP_UIKit_StatusIndicator
   properties {type=Attention, bangType=SocialButton, iconSize=Compact}
   slot {class=GridSlot, properties {horizontalAlignment=HAlign_Right, layer=5, verticalAlignment=VAlign_Top, nudge=<3 -3>}}
  }
  Overlay_112 {widgetClassPath=&Overlay, widget=Overlay_112}
  WBP_CaptureForPostBufferUpdate {
   widgetClassPath=&WBP_CaptureForPostBufferUpdate
   bIsVariable=true
   parent=Overlay_112
   slot {class=OverlaySlot, properties {horizontalAlignment=HAlign_Fill, verticalAlignment=VAlign_Fill}}
  }
  Backplate {
   widgetClassPath=&WBP_UIKit_Backplate
   bIsVariable=true
   parent=Overlay_112
   widget=Backplate
   properties {brightness=Dark, cornerRadius=Circle}
   slot {class=OverlaySlot, properties {horizontalAlignment=HAlign_Fill, verticalAlignment=VAlign_Fill}}
  }
 ]
}
