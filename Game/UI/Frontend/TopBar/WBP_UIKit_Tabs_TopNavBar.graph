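# Frontend top navigation bar tab widget, a child of WBP_UIKit_Tabs_Base with no widgets or variables of its own.
# DetermineDividerPositionThenSetTabsData casts each incoming tab view model to FortTabViewModel, setting TabID from its linked UI feature and SubGroupID from its button group, both as enum names.
# It then passes the array to Set Tabs Data, with no explicit divider math here despite the function name.
# The event graph only calls parent handlers.
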WidgetBlueprint /Game/UI/Frontend/TopBar/WBP_UIKit_Tabs_TopNavBar
graphs [
 DetermineDividerPositionThenSetTabsData {nodes [
   n1 = DetermineDividerPositionThenSetTabsData!() -> (then, InTabsData:Array of UI Kit Tab Group Button View Model Object References) {is=title K2Node_FunctionEntry}
   > n2 = For Each Loop(Array:Array of UI Kit Selectable Button View Model Object References=@n1) -> (Array Element:UI Kit Selectable Button View Model Object Reference)
   Set Tabs Data!(
    execute=@n2.Completed
    self:WBP UIKit Tabs Base Object Reference
    InTabs Data:Array of UI Kit Tab Group Button View Model Object References=@n1
   ) -> (then) {is=title K2Node_CallFunction}
   n3 = Cast To FortTabViewModel!(execute=@n2.LoopBody, Object=@n2) -> (then, CastFailed, AsFort Tab Viewmodel:Fort Tab Viewmodel Object Reference) {is=title K2Node_DynamicCast}
   > Set TabID!(
    execute
    TabID:Name=Enum to Name!(Enumerator:EFortUIFeature Enum=GetLinkedUIFeature!(self:Fort Tab Viewmodel Object Reference=@n3) -> (ReturnValue:EFortUIFeature Enum=ShowHome)) -> (ReturnValue:Name) {is=title K2Node_GetEnumeratorName}
    self:UI Kit Tab Group Button View Model Object Reference=@n3
   ) -> (then, Output_Get:Name=None)
   > Set SubGroupID!(
    execute
    SubGroupID:Name=Enum to Name(Enumerator:EFortTopBarTabButtonGroup Enum=GetButtonGroup!(self:Fort Tab Viewmodel Object Reference=@n3) -> (ReturnValue:EFortTopBarTabButtonGroup Enum=Discover))
    self:UI Kit Tab Group Button View Model Object Reference=@n3
   ) -> (then, Output_Get:Name=None)
  ]}
 EventGraph {nodes [
   n1 = Event PreConstruct()
   > Parent: PreConstruct(IsDesignTime=@n1.IsDesignTime)
   Event Construct()
   > Parent: Construct()
   Event Destruct()
   > Parent: Destruct()
  ]}
]
variables []
widgets {
 info {
  inheritedWidgetCount=0
  parentClass=/FortUILibrary/Systems/Tabs/WBP_UIKit_Tabs_Base.WBP_UIKit_Tabs_Base_C
  rootWidgetClass=null
  widgetCount=0
 }
 widgets []
}
